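# Niagara system for the HexFade v2 vehicle boost cosmetic, trailing ribbons and glowing sprites from skeletal-mesh boost sockets.
# BlueTrails001 and RedTrails001 spawn per-port ribbons, writing four age-driven sine offset "noodle" positions and separate fading colors for multiple renderers.
# HexRed spawns two VisTag-tagged groups of short-lived SubUV sprites, partitioned across sockets and flung backward.
# VC_PaintedVariantsOverride supplies each emitter's primary and secondary paint-variant colors.
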NiagaraSystem /VehicleCosmetics/Boosts/Boost_HexFade/FX/NS_Boost_HexFade_v2
dependencies {
 usedDataInterfaces [
  &NiagaraDataInterfaceCurve
  &NiagaraDataInterfaceSpriteRendererInfo
  &NiagaraDataInterfaceSkeletalMesh
 ]
 usedDynamicInputs []
 usedModules [
  &self:SystemSpawnScript.SystemScriptSource.SystemScriptGraph.NiagaraNodeAssignment_0.SetVariables_#1
  &SystemState
 ]
}
emitters {
 BlueTrails001 {
  emitterUpdateScript [
   &EmitterState()
   &SpawnPerUnit(Module.Use Max Movement Threshold=true, Module.Use Movement Tolerance=false, Use Spawn Probability=false) {enabled=false}
   &VC_FilteredSocketCount()
   &VC_RateQualityHandler()
   &VC_SpawnRibbon_v1(Number of Ribbons=$Emitter.PortCount, SpawnRate=60)
  ]
  particleSpawnScript [
   &VC_PaintedVariantsOverride(
    Black=<rgba 0.0156 0.0156 0.0156 1>
    BurntSienna=<rgba 0.1562 0.0504 0.022 0.8>
    Cobalt=<rgba 0.0522 0.138 0.9115 0.8348>
    Crimson=<rgba 0.4271 0.0356 0.0462 0.5217>
    DefaultColor=<rgba 0.1284 0.3065 1 1>
    ForestGreen=<rgba 0 0.5 0.0253 0.4>
    Grey=<rgba 0.2083 0.2083 0.2083 1>
    Lime=<rgba 0.4075 0.6094 0.0508 0.8522>
    Orange=<rgba 0.6724 0.2346 0 0.8>
    Pink=<rgba 0.7 0.1896 0.5144 0.8>
    Purple=<rgba 0.1913 0 0.6795 1>
    R1 Black:LinearColor=<rgba 0.2708 0.2708 0.2708 1>
    R1 BurntSienna:LinearColor=<rgba 0.8229 0.2696 0.0129 1>
    R1 Cobalt:LinearColor=<rgba 0 0.185 10 1>
    R1 Crimson:LinearColor=<rgba 1 0 0.0271 1>
    R1 DefaultColor=<rgba 4 0.28 0.28 1>
    R1 Grey:LinearColor=<rgba 0.7478 0.7478 0.7478 1>
    R1 Lime:LinearColor=<rgba 1.5 3 0 1>
    R1 Orange:LinearColor=<rgba 5 1.3346 0 1>
    R1 Pink:LinearColor=<rgba 3 0.5 2.0909 1>
    R1 Purple:LinearColor=<rgba 0.4 0 2 1>
    R1 Saffron:LinearColor=<rgba 3 2.3 0 1>
    R1 SkyBlue:LinearColor=<rgba 0 1.9962 3 1>
    R1 TitaniumWhite:LinearColor=<rgba 1 1 1 1>
    R2 Black:LinearColor=<rgba 0.0365 0.0365 0.0365 1>
    R2 BurntSienna:LinearColor=<rgba 0.2083 0.0709 0.0335 1>
    R2 DefaultColor:LinearColor=<rgba 0.1284 0.3065 1 1>
    Saffron=<rgba 0.8698 0.6774 0.0453 0.6957>
    SkyBlue=<rgba 0 0.6654 1 1>
    TitaniumWhite=<rgba 0.8854 0.8854 0.8854 1>
    Use R1 Painted Overrides=true
    Use R2 Painted Overrides=true
   )
   &VC_InitializeRibbon_v1(
    Color=$Particles.Color
    Enable Skeletal Position=true
    Lifetime=0.6
    Offset Amount=-50
    Offset Axis:ENiagaraOrientationAxis=::XAxis
    Ribbon Width=150
   )
   &AddVelocity(Velocity=<-1200 0 0>, Velocity Speed Scale=1.4) {module=AddVelocity001}
  ]
  particleUpdateScript [
   &ParticleState()
   &InheritSourceMovement(
    Applied Position Scale=&Multiply_VectorByFloat(
     Vector=<1 1 1>
     Float=&FloatFromCurve(FloatCurve=DataInterface() {propertyValues {Curve {
         defaultValue=1
         keys [{interpMode=RCIM_Cubic, tangentMode=RCTM_User, time=0, value=1}, {interpMode=RCIM_Cubic, tangentMode=RCTM_User, time=1, value=0}]
        }}})
    )
    ApplyForce=false
    ApplyPosition=true
    ApplyPositionFalloff=false
   )
   &ScaleRibbonWidth(Ribbon Width Scale=&FloatFromCurve(FloatCurve=DataInterface() {propertyValues {Curve {
        defaultValue=1
        keys [{interpMode=RCIM_Cubic, tangentMode=RCTM_User, time=0, value=1}, {interpMode=RCIM_Cubic, tangentMode=RCTM_User, time=1, value=0}]
       }}}))
   &SolveForcesAndVelocity(Clamp Velocity=false, Limit Acceleration=false)
   &VC_UpdateRibbon_v1(Enable Skeletal Position=true)
   /SetVariables(
    Particles.PosNoodle1:NiagaraPosition=&AddVectorToPosition(
     Position=$Particles.Position
     Vector=&TransformVector(
      Vector=&MakeVector(
       X=0
       Y=&Sine(
        Normalized Angle=&Add_Float(
         A=&Multiply_Float(A=$Particles.NormalizedAge, B=-1.3)
         B=&Multiply_Float(A=$Emitter.LoopedAge, B=1.8)
        )
        Period=0.5
        Scale=&FloatFromCurve(FloatCurve=DataInterface() {propertyValues {Curve {defaultValue=0.3, keys [<at 0 1>, {tangentMode=RCTM_User, time=0.495, value=1}, {tangentMode=RCTM_User, time=1, value=40}]}}})
        Bias=0
       )
       Z=&Sine(
        Normalized Angle=&Add_Float(
         A=&Multiply_Float(A=$Particles.NormalizedAge, B=-1.3)
         B=&Multiply_Float(A=$Emitter.LoopedAge, B=2)
        )
        Period=0.6
        Scale=&FloatFromCurve(FloatCurve=DataInterface() {propertyValues {Curve {defaultValue=0.3, keys [<at 0 0>, {tangentMode=RCTM_User, time=0.6263, value=35.8247}, {tangentMode=RCTM_User, time=1, value=60}]}}})
        Bias=0
       )
      )
      Destination Space=::Simulation
      Apply Scale=true
     )
    )
    Particles.PosNoodle2:NiagaraPosition=&AddVectorToPosition(
     Position=$Particles.Position
     Vector=&TransformVector(
      Vector=&MakeVector(
       X=0
       Y=&Sine(
        Normalized Angle=&Add_Float(
         A=&Multiply_Float(A=$Particles.NormalizedAge, B=-1.3)
         B=&Multiply_Float(A=$Emitter.LoopedAge, B=1.8)
        )
        Period=0.4
        Scale=&FloatFromCurve(FloatCurve=DataInterface() {propertyValues {Curve {keys [
             {arriveTangent=41.4286, interpMode=RCIM_Cubic, time=0, value=0.3}
             {arriveTangentWeight=3.3375, interpMode=RCIM_Cubic, leaveTangentWeight=3.3375, tangentMode=RCTM_User, time=0.3, value=20}
             {arriveTangent=28.1747, interpMode=RCIM_Cubic, leaveTangent=28.1747, tangentMode=RCTM_User, time=1, value=50}
            ]}}})
        Bias=0
       )
       Z=&Sine(
        Normalized Angle=&Add_Float(
         A=&Multiply_Float(A=$Particles.NormalizedAge, B=-1.3)
         B=&Multiply_Float(A=$Emitter.LoopedAge, B=2)
        )
        Period=0.8
        Scale=&FloatFromCurve(FloatCurve=DataInterface() {propertyValues {Curve {keys [
             <at 0.3 0.3 RCIM_Cubic>
             {
              arriveTangent=10.3659
              arriveTangentWeight=3.3375
              interpMode=RCIM_Cubic
              leaveTangent=10.3659
              leaveTangentWeight=3.3375
              tangentMode=RCTM_User
              time=0.5
              value=20
             }
             {arriveTangent=75.1908, interpMode=RCIM_Cubic, leaveTangent=75.1908, tangentMode=RCTM_User, time=1, value=60}
            ]}}})
        Bias=0
       )
      )
      Destination Space=::Simulation
      Apply Scale=true
     )
    )
    Particles.PosNoodle3:NiagaraPosition=&AddVectorToPosition(
     Position=$Particles.Position
     Vector=&TransformVector(
      Vector=&MakeVector(
       X=0
       Y=&Sine(
        Normalized Angle=&Add_Float(
         A=&Multiply_Float(A=$Particles.NormalizedAge, B=-1.3)
         B=&Multiply_Float(A=$Emitter.LoopedAge, B=2.1)
        )
        Period=0.5
        Scale=&FloatFromCurve(FloatCurve=DataInterface() {propertyValues {Curve {keys [
             {arriveTangent=39.5768, interpMode=RCIM_Cubic, leaveTangent=39.5768, tangentMode=RCTM_User, time=0, value=0.3}
             {
              arriveTangent=32.917
              arriveTangentWeight=3.3375
              interpMode=RCIM_Cubic
              leaveTangent=32.917
              leaveTangentWeight=3.3375
              time=0.595
              value=8.803
             }
             {interpMode=RCIM_Cubic, leaveTangent=75.1908, time=1, value=33.217}
            ]}}})
        Bias=0
       )
       Z=&Sine(
        Normalized Angle=&Add_Float(
         A=&Multiply_Float(A=$Particles.NormalizedAge, B=-1.3)
         B=&Multiply_Float(A=$Emitter.LoopedAge, B=1.7)
        )
        Period=0.5
        Scale=&FloatFromCurve(FloatCurve=DataInterface() {propertyValues {Curve {keys [
             <at 0 0 RCIM_Cubic>
             {arriveTangentWeight=3.3375, interpMode=RCIM_Cubic, leaveTangentWeight=3.3375, tangentMode=RCTM_User, time=0.3, value=1}
             {arriveTangent=75.1908, interpMode=RCIM_Cubic, leaveTangent=75.1908, tangentMode=RCTM_User, time=1, value=45}
            ]}}})
        Bias=0
       )
      )
      Destination Space=::Simulation
      Apply Scale=true
     )
    )
    Particles.PosNoodle4:NiagaraPosition=&AddVectorToPosition(
     Position=$Particles.Position
     Vector=&TransformVector(
      Vector=&MakeVector(
       X=0
       Y=&Sine(
        Normalized Angle=&Add_Float(
         A=&Multiply_Float(A=$Particles.NormalizedAge, B=-1.3)
         B=&Multiply_Float(A=$Emitter.LoopedAge, B=1.5)
        )
        Period=0.4
        Scale=&FloatFromCurve(FloatCurve=DataInterface() {propertyValues {Curve {keys [
             {arriveTangent=41.4286, interpMode=RCIM_Cubic, time=0, value=0.3}
             {arriveTangentWeight=3.3375, interpMode=RCIM_Cubic, leaveTangentWeight=3.3375, tangentMode=RCTM_User, time=0.2, value=11.4906}
             {arriveTangent=31.4052, interpMode=RCIM_Cubic, leaveTangent=31.4052, tangentMode=RCTM_User, time=1, value=26}
            ]}}})
        Bias=0
       )
       Z=&Sine(
        Normalized Angle=&Add_Float(
         A=&Multiply_Float(A=$Particles.NormalizedAge, B=-1.3)
         B=&Multiply_Float(A=$Emitter.LoopedAge, B=1)
        )
        Period=1
        Scale=&FloatFromCurve(FloatCurve=DataInterface() {propertyValues {Curve {keys [
             <at 0 1 RCIM_Cubic>
             {arriveTangentWeight=6.6125, interpMode=RCIM_Cubic, leaveTangentWeight=6.725, tangentMode=RCTM_User, time=0.6, value=20}
             {interpMode=RCIM_Cubic, leaveTangent=75.1908, time=1, value=35}
            ]}}})
        Bias=0
       )
      )
      Destination Space=::Simulation
      Apply Scale=true
     )
    )
   ) {isSetParameters=true, module=SetVariables_#2}
   /SetVariables(
    Particles.RibbonColor0:LinearColor=/DealershipSharedLibrary/FX/Niagara/DynamicInputs/MultiplyLinearColorByVectorAndFloat(
     Color to Scale:LinearColor=$Particles.Initial.Color
     RGB Scale:Vector3f=&VectorFromFloat(Value=&FloatFromCurve(FloatCurve=DataInterface() {propertyValues {Curve {keys [<at 0 1>,<at 1 0>]}}}))
     Alpha Scale:NiagaraFloat=&FloatFromCurve(FloatCurve=DataInterface() {propertyValues {Curve {
         defaultValue=1
         keys [{interpMode=RCIM_Cubic, tangentMode=RCTM_User, time=0.3, value=0.5}, {interpMode=RCIM_Cubic, tangentMode=RCTM_User, time=1, value=0}]
        }}})
    )
    Particles.RibbonColor1:LinearColor=&MultiplyLinearColorByVectorAndFloat(
     Color to Scale:LinearColor=$Particles.Initial.Color
     RGB Scale:Vector3f=&VectorFromFloat(Value=&MakeCustomFloatFromBool(Bool=$User.bIsPainted, True Float=1, False Float=2))
     Alpha Scale:NiagaraFloat=&FloatFromCurve(FloatCurve=DataInterface() {propertyValues {Curve {
         defaultValue=1
         keys [{interpMode=RCIM_Cubic, tangentMode=RCTM_User, time=0.3, value=0.5}, {interpMode=RCIM_Cubic, tangentMode=RCTM_User, time=1, value=0}]
        }}})
    )
   ) {isSetParameters=true, module=SetVariables_#3}
  ]
  rendererClasses [&NiagaraRibbonRendererProperties]
  renderers [
   {rendererClass=&NiagaraRibbonRendererProperties}
   {rendererClass=&NiagaraRibbonRendererProperties, rendererIndex=1}
   {rendererClass=&NiagaraRibbonRendererProperties, rendererIndex=2}
   {rendererClass=&NiagaraRibbonRendererProperties, rendererIndex=3}
  ]
 }
 HexRed {
  emitterUpdateScript [
   &EmitterState()
   &VC_FilteredSocketCount()
   &SpawnRate(SpawnRate=80, Use Spawn Probability=false)
   &SpawnRate(SpawnGroup=1, SpawnRate=80, Use Spawn Probability=false) {module=SpawnRate001}
  ]
  particleSpawnScript [
   /SetVariables(Particles.VisTag:NiagaraInt32=$Emitter.SpawnGroup) {isSetParameters=true}
   &VC_PaintedVariantsOverride(
    Black=<rgba 0.1302 0.1302 0.1302 1>
    BurntSienna=<rgba 0.2552 0.0865 0.0412 1>
    Cobalt=<rgba 0 0.0555 3 1>
    Crimson=<rgba 1 0 0.0271 1>
    DefaultColor=<rgba 0 0.2044 1 1>
    Pink=<rgba 0.9844 0.1179 0.6693 1>
    R1 Black:LinearColor=<rgba 0.1 0.1 0.1 5>
    R1 BurntSienna:LinearColor=<rgba 0.7708 0.2624 0.1241 1>
    R1 Cobalt:LinearColor=<rgba 0 0.0555 3 1>
    R1 Crimson:LinearColor=<rgba 1 0 0.0271 1>
    R1 DefaultColor=<rgba 1 0.1373 0.1373 1>
    R1 Pink:LinearColor=<rgba 0.875 0.1048 0.5949 1>
    Use R1 Painted Overrides=true
   )
   &InitializeParticle(
    Color Mode=::NewEnumerator0
    Lifetime=&RandomRangeFloat(Minimum=0.2, Maximum=0.3)
    Position Offset=<300 0 0>
    Sprite Rotation Mode=::NewEnumerator2
    Sprite Size Mode=::NewEnumerator1
    Uniform Sprite Size Max=15
    Uniform Sprite Size Min=40
    UsePositionOffset=false
    Write Material Random=true
   )
   &PartitionParticles(Number Of Partitions=$Emitter.PortCount, Particles Per Partition=3, Partition Mode:ENiagara_PartitionMode=::NewEnumerator1)
   &SkeletalMeshLocation(
    Bone / Socket Index=$Particles.PartitionParticles.Partition
    Bone / Socket Sampling Mode=::NewEnumerator5
    MaskSpawnGroup=false
    Mesh Sampling Type=::NewEnumerator6
    Skeletal Mesh=$System.SkeletalMesh
   )
   &AddVelocity(Velocity=<-1500 0 0>, Velocity Speed Scale=2)
   &ShapeLocation(MaskSpawnGroup=false, Override Seed=false, Sphere Radius=30, UseOwnerScale=true) {module=ShapeLocation001}
  ]
  particleUpdateScript [
   &ParticleState()
   &ScaleSpriteSize(Uniform Curve Sprite Scale=DataInterface() {propertyValues {Curve {
       defaultValue=0
       keys [
        {tangentMode=RCTM_User, time=0, value=1}
        {arriveTangent=-0.0405, interpMode=RCIM_Cubic, leaveTangent=-0.0405, tangentMode=RCTM_User, time=0.7009, value=1}
        {arriveTangent=-8.0089, interpMode=RCIM_Cubic, leaveTangent=-8.0089, tangentMode=RCTM_User, time=1, value=0}
       ]
      }}})
   &InheritSourceMovement(
    Applied Position Scale=&Multiply_VectorByFloat(
     Vector=<1 1 1>
     Float=&FloatFromCurve(FloatCurve=DataInterface() {propertyValues {Curve {keys [<at 0.1 1 RCIM_Cubic>, {arriveTangent=-3.4326, interpMode=RCIM_Cubic, leaveTangent=-3.4326, tangentMode=RCTM_User, time=0.9, value=0}]}}})
    )
    ApplyForce=false
    ApplyPosition=true
    ApplyPositionFalloff=false
   )
   &SpriteRotationRate()
   &Drag(
    Drag=&FloatFromCurve(
     FloatCurve=DataInterface() {propertyValues {Curve {keys [{arriveTangent=-7.047, interpMode=RCIM_Cubic, leaveTangent=-7.047, tangentMode=RCTM_User, time=0, value=5}, <at 0.5 -20>]}}}
     Scale Curve=5
    )
    Use Linear Drag=true
    Use Rotational Drag=false
   )
   &SubUVAnimation(
    End Frame Range Override=7
    "Sprite Renderer "=DataInterface() {propertyValues {SpriteRenderer=&self:HexRed_0.NiagaraSpriteRendererProperties_3}}
    UseEndFrame=true
    UseStartFrame=true
   )
   &SubUVAnimation(
    End Frame Range Override=7
    "Sprite Renderer "=DataInterface() {propertyValues {SpriteRenderer=&self:HexRed_0.NiagaraSpriteRendererProperties_1}}
    SubUV Animation Mode:ENiagara_SubUVLookupModeV2=::NewEnumerator1
    SubUV Curve=DataInterface() {propertyValues {Curve {
       defaultValue=0
       keys [
        {arriveTangent=1.815, interpMode=RCIM_Cubic, leaveTangent=1.815, tangentMode=RCTM_User, time=0, value=0}
        {interpMode=RCIM_Cubic, tangentMode=RCTM_User, time=1, value=1}
       ]
      }}}
    UseEndFrame=true
    UseStartFrame=true
   ) {module=SubUVAnimation001}
   &SolveForcesAndVelocity(Clamp Velocity=false, Limit Acceleration=false)
   /SetVariables(
    Particles.SpriteColor0:LinearColor=&MultiplyLinearColorByVectorAndFloat(
     Color to Scale:LinearColor=$Particles.Color
     RGB Scale:Vector3f=&VectorFromFloat(Value=&FloatFromCurve(FloatCurve=DataInterface() {propertyValues {Curve {
          defaultValue=1
          keys [
           {interpMode=RCIM_Cubic, tangentMode=RCTM_User, time=0, value=1}
           {interpMode=RCIM_Cubic, tangentMode=RCTM_User, time=0.3, value=20}
           {tangentMode=RCTM_User, time=1, value=1}
          ]
         }}}))
     Alpha Scale:NiagaraFloat=&FloatFromCurve(FloatCurve=DataInterface() {propertyValues {Curve {
         defaultValue=0
         keys [
          {interpMode=RCIM_Cubic, tangentMode=RCTM_User, time=0, value=0}
          {interpMode=RCIM_Cubic, tangentMode=RCTM_User, time=0.5, value=1}
          {tangentMode=RCTM_User, time=0.7, value=0}
         ]
        }}})
    )
    Particles.SpriteColor1:LinearColor=&MultiplyLinearColorByVectorAndFloat(
     Color to Scale:LinearColor=$Particles.Initial.R1 Color
     RGB Scale:Vector3f=&VectorFromFloat(Value=&FloatFromCurve(FloatCurve=DataInterface() {propertyValues {Curve {
          defaultValue=1
          keys [
           {interpMode=RCIM_Cubic, tangentMode=RCTM_User, time=0, value=2}
           {interpMode=RCIM_Cubic, tangentMode=RCTM_User, time=0.3, value=20}
           {tangentMode=RCTM_User, time=1, value=2}
          ]
         }}}))
     Alpha Scale:NiagaraFloat=&FloatFromCurve(FloatCurve=DataInterface() {propertyValues {Curve {
         defaultValue=0
         keys [
          {interpMode=RCIM_Cubic, tangentMode=RCTM_User, time=0, value=0}
          {interpMode=RCIM_Cubic, tangentMode=RCTM_User, time=0.5, value=1}
          {tangentMode=RCTM_User, time=0.7, value=0}
         ]
        }}})
    )
   ) {isSetParameters=true}
  ]
  rendererClasses [&NiagaraSpriteRendererProperties]
  renderers [
   {rendererClass=&NiagaraSpriteRendererProperties}
   {rendererClass=&NiagaraSpriteRendererProperties, rendererIndex=1}
  ]
 }
 RedTrails001 {
  emitterUpdateScript [
   &EmitterState()
   &SpawnPerUnit(Module.Use Max Movement Threshold=true, Module.Use Movement Tolerance=false, Use Spawn Probability=false) {enabled=false}
   &VC_FilteredSocketCount()
   &VC_RateQualityHandler()
   &VC_SpawnRibbon_v1(Number of Ribbons=$Emitter.PortCount, SpawnRate=60)
  ]
  particleSpawnScript [
   &VC_PaintedVariantsOverride(
    Black=<rgba 0.1667 0.1667 0.1667 1>
    BurntSienna=<rgba 0.1562 0.0504 0.022 0.8>
    Cobalt=<rgba 0.0522 0.138 0.9115 0.8348>
    Crimson=<rgba 0.4271 0.0356 0.0462 0.5217>
    DefaultColor=<rgba 0.1284 0.3065 1 1>
    ForestGreen=<rgba 0 0.5 0.0253 0.4>
    Grey=<rgba 0.5469 0.5469 0.5469 1>
    Lime=<rgba 0.3906 0.75 0.0313 1>
    Orange=<rgba 0.6724 0.2346 0 0.8>
    Pink=<rgba 0.7 0.1896 0.5144 0.8>
    Purple=<rgba 0.1913 0 0.6795 1>
    R1 Black:LinearColor=<rgba 0.1458 0.1458 0.1458 1>
    R1 BurntSienna:LinearColor=<rgba 0.8229 0.2696 0.0129 1>
    R1 Cobalt:LinearColor=<rgba 0 0.185 10 1>
    R1 Crimson:LinearColor=<rgba 1 0.0938 0.1183 1>
    R1 DefaultColor=<rgba 2 0.5625 0.5625 1>
    R1 ForestGreen:LinearColor=<rgba 0.0875 0.6719 0.117 1>
    R1 Grey:LinearColor=<rgba 0.3177 0.3177 0.3177 1>
    R1 Lime:LinearColor=<rgba 1.1093 1.5469 0 1>
    R1 Orange:LinearColor=<rgba 5 1.3346 0 1>
    R1 Pink:LinearColor=<rgba 3 0.5 2.0909 1>
    R1 Purple:LinearColor=<rgba 0.4 0 2 1>
    R1 Saffron:LinearColor=<rgba 1.3096 1.004 0 1>
    R1 SkyBlue:LinearColor=<rgba 0 1.9962 3 1>
    R1 TitaniumWhite:LinearColor=<rgba 1 1 1 1>
    R2 Black:LinearColor=<rgba 0.0365 0.0365 0.0365 1>
    R2 BurntSienna:LinearColor=<rgba 0.2083 0.0709 0.0335 1>
    R2 DefaultColor:LinearColor=<rgba 0.1284 0.3065 1 1>
    Saffron=<rgba 0.9 0.7009 0.0469 1>
    SkyBlue=<rgba 0 0.6654 1 1>
    TitaniumWhite=<rgba 1 1 1 1>
    Use R1 Painted Overrides=true
    Use R2 Painted Overrides=true
   )
   &VC_InitializeRibbon_v1(
    Color=$Particles.Color
    Enable Skeletal Position=true
    Lifetime=0.6
    Offset Amount=-50
    Offset Axis:ENiagaraOrientationAxis=::XAxis
    Ribbon Width=150
   )
   &AddVelocity(Velocity=<-1200 0 0>, Velocity Speed Scale=1.4) {module=AddVelocity001}
  ]
  particleUpdateScript [
   &ParticleState()
   &InheritSourceMovement(
    Applied Position Scale=&Multiply_VectorByFloat(
     Vector=<1 1 1>
     Float=&FloatFromCurve(FloatCurve=DataInterface() {propertyValues {Curve {
         defaultValue=1
         keys [{interpMode=RCIM_Cubic, tangentMode=RCTM_User, time=0, value=1}, {interpMode=RCIM_Cubic, tangentMode=RCTM_User, time=1, value=0}]
        }}})
    )
    ApplyForce=false
    ApplyPosition=true
    ApplyPositionFalloff=false
   )
   &ScaleRibbonWidth(Ribbon Width Scale=&FloatFromCurve(FloatCurve=DataInterface() {propertyValues {Curve {
        defaultValue=1
        keys [{interpMode=RCIM_Cubic, tangentMode=RCTM_User, time=0, value=1}, {interpMode=RCIM_Cubic, tangentMode=RCTM_User, time=1, value=0}]
       }}}))
   &SolveForcesAndVelocity(Clamp Velocity=false, Limit Acceleration=false)
   &VC_UpdateRibbon_v1(Enable Skeletal Position=true)
   /SetVariables(
    Particles.PosNoodle1:NiagaraPosition=&AddVectorToPosition(
     Position=$Particles.Position
     Vector=&TransformVector(
      Vector=&MakeVector(
       X=0
       Y=&Sine(
        Normalized Angle=&Add_Float(
         A=&Multiply_Float(A=$Particles.NormalizedAge, B=-1.3)
         B=&Multiply_Float(A=$Emitter.LoopedAge, B=1.8)
        )
        Period=0.5
        Scale=&FloatFromCurve(FloatCurve=DataInterface() {propertyValues {Curve {defaultValue=0.3, keys [<at 0 1>, {tangentMode=RCTM_User, time=0.495, value=1}, {tangentMode=RCTM_User, time=1, value=40}]}}})
        Bias=0
       )
       Z=&Sine(
        Normalized Angle=&Add_Float(
         A=&Multiply_Float(A=$Particles.NormalizedAge, B=-1.3)
         B=&Multiply_Float(A=$Emitter.LoopedAge, B=2)
        )
        Period=0.6
        Scale=&FloatFromCurve(FloatCurve=DataInterface() {propertyValues {Curve {defaultValue=0.3, keys [<at 0 0>, {tangentMode=RCTM_User, time=0.6263, value=35.8247}, {tangentMode=RCTM_User, time=1, value=60}]}}})
        Bias=0
       )
      )
      Destination Space=::Simulation
      Apply Scale=true
     )
    )
    Particles.PosNoodle2:NiagaraPosition=&AddVectorToPosition(
     Position=$Particles.Position
     Vector=&TransformVector(
      Vector=&MakeVector(
       X=0
       Y=&Sine(
        Normalized Angle=&Add_Float(
         A=&Multiply_Float(A=$Particles.NormalizedAge, B=-1.3)
         B=&Multiply_Float(A=$Emitter.LoopedAge, B=1.8)
        )
        Period=0.4
        Scale=&FloatFromCurve(FloatCurve=DataInterface() {propertyValues {Curve {keys [
             {arriveTangent=41.4286, interpMode=RCIM_Cubic, time=0, value=0.3}
             {arriveTangentWeight=3.3375, interpMode=RCIM_Cubic, leaveTangentWeight=3.3375, tangentMode=RCTM_User, time=0.3, value=20}
             {arriveTangent=28.1747, interpMode=RCIM_Cubic, leaveTangent=28.1747, tangentMode=RCTM_User, time=1, value=50}
            ]}}})
        Bias=0
       )
       Z=&Sine(
        Normalized Angle=&Add_Float(
         A=&Multiply_Float(A=$Particles.NormalizedAge, B=-1.3)
         B=&Multiply_Float(A=$Emitter.LoopedAge, B=2)
        )
        Period=0.8
        Scale=&FloatFromCurve(FloatCurve=DataInterface() {propertyValues {Curve {keys [
             <at 0.3 0.3 RCIM_Cubic>
             {
              arriveTangent=10.3659
              arriveTangentWeight=3.3375
              interpMode=RCIM_Cubic
              leaveTangent=10.3659
              leaveTangentWeight=3.3375
              tangentMode=RCTM_User
              time=0.5
              value=20
             }
             {arriveTangent=75.1908, interpMode=RCIM_Cubic, leaveTangent=75.1908, tangentMode=RCTM_User, time=1, value=60}
            ]}}})
        Bias=0
       )
      )
      Destination Space=::Simulation
      Apply Scale=true
     )
    )
    Particles.PosNoodle3:NiagaraPosition=&AddVectorToPosition(
     Position=$Particles.Position
     Vector=&TransformVector(
      Vector=&MakeVector(
       X=0
       Y=&Sine(
        Normalized Angle=&Add_Float(
         A=&Multiply_Float(A=$Particles.NormalizedAge, B=-1.3)
         B=&Multiply_Float(A=$Emitter.LoopedAge, B=2.1)
        )
        Period=0.5
        Scale=&FloatFromCurve(FloatCurve=DataInterface() {propertyValues {Curve {keys [
             {arriveTangent=39.5768, interpMode=RCIM_Cubic, leaveTangent=39.5768, tangentMode=RCTM_User, time=0, value=0.3}
             {
              arriveTangent=32.917
              arriveTangentWeight=3.3375
              interpMode=RCIM_Cubic
              leaveTangent=32.917
              leaveTangentWeight=3.3375
              time=0.595
              value=8.803
             }
             {interpMode=RCIM_Cubic, leaveTangent=75.1908, time=1, value=33.217}
            ]}}})
        Bias=0
       )
       Z=&Sine(
        Normalized Angle=&Add_Float(
         A=&Multiply_Float(A=$Particles.NormalizedAge, B=-1.3)
         B=&Multiply_Float(A=$Emitter.LoopedAge, B=1.7)
        )
        Period=0.5
        Scale=&FloatFromCurve(FloatCurve=DataInterface() {propertyValues {Curve {keys [
             <at 0 0 RCIM_Cubic>
             {arriveTangentWeight=3.3375, interpMode=RCIM_Cubic, leaveTangentWeight=3.3375, tangentMode=RCTM_User, time=0.3, value=1}
             {arriveTangent=75.1908, interpMode=RCIM_Cubic, leaveTangent=75.1908, tangentMode=RCTM_User, time=1, value=45}
            ]}}})
        Bias=0
       )
      )
      Destination Space=::Simulation
      Apply Scale=true
     )
    )
    Particles.PosNoodle4:NiagaraPosition=&AddVectorToPosition(
     Position=$Particles.Position
     Vector=&TransformVector(
      Vector=&MakeVector(
       X=0
       Y=&Sine(
        Normalized Angle=&Add_Float(
         A=&Multiply_Float(A=$Particles.NormalizedAge, B=-1.3)
         B=&Multiply_Float(A=$Emitter.LoopedAge, B=1.5)
        )
        Period=0.4
        Scale=&FloatFromCurve(FloatCurve=DataInterface() {propertyValues {Curve {keys [
             {arriveTangent=41.4286, interpMode=RCIM_Cubic, time=0, value=0.3}
             {arriveTangentWeight=3.3375, interpMode=RCIM_Cubic, leaveTangentWeight=3.3375, tangentMode=RCTM_User, time=0.2, value=11.4906}
             {arriveTangent=31.4052, interpMode=RCIM_Cubic, leaveTangent=31.4052, tangentMode=RCTM_User, time=1, value=26}
            ]}}})
        Bias=0
       )
       Z=&Sine(
        Normalized Angle=&Add_Float(
         A=&Multiply_Float(A=$Particles.NormalizedAge, B=-1.3)
         B=&Multiply_Float(A=$Emitter.LoopedAge, B=1)
        )
        Period=1
        Scale=&FloatFromCurve(FloatCurve=DataInterface() {propertyValues {Curve {keys [
             <at 0 1 RCIM_Cubic>
             {arriveTangentWeight=6.6125, interpMode=RCIM_Cubic, leaveTangentWeight=6.725, tangentMode=RCTM_User, time=0.6, value=20}
             {interpMode=RCIM_Cubic, leaveTangent=75.1908, time=1, value=35}
            ]}}})
        Bias=0
       )
      )
      Destination Space=::Simulation
      Apply Scale=true
     )
    )
   ) {isSetParameters=true, module=SetVariables_#2}
   /SetVariables(
    Particles.RibbonColor2:LinearColor=&MultiplyLinearColorByVectorAndFloat(
     Color to Scale:LinearColor=$Particles.Initial.R1 Color
     RGB Scale:Vector3f=&VectorFromFloat(Value=&MakeFloatFromVector(
       Vector=&VectorFromFloat(Value=&FloatFromCurve(FloatCurve=DataInterface() {propertyValues {Curve {keys [<at 0 1>,<at 1 0>]}}}))
       Channel=::NewEnumerator0
      ))
     Alpha Scale:NiagaraFloat=&FloatFromCurve(FloatCurve=DataInterface() {propertyValues {Curve {
         defaultValue=1
         keys [
          {arriveTangent=-2.5302, interpMode=RCIM_Cubic, leaveTangent=-2.5302, tangentMode=RCTM_User, time=0, value=1}
          {interpMode=RCIM_Cubic, tangentMode=RCTM_User, time=1, value=0.1}
         ]
        }}})
    )
    Particles.RibbonColor2B:LinearColor=&MultiplyLinearColorByVectorAndFloat(
     Color to Scale:LinearColor=$Particles.Initial.R1 Color
     RGB Scale:Vector3f=&VectorFromFloat(Value=&MakeCustomFloatFromBool(Bool=$User.bIsPainted, True Float=5, False Float=5))
     Alpha Scale:NiagaraFloat=&FloatFromCurve(FloatCurve=DataInterface() {propertyValues {Curve {
         defaultValue=1
         keys [{interpMode=RCIM_Cubic, tangentMode=RCTM_User, time=0, value=1}, {interpMode=RCIM_Cubic, tangentMode=RCTM_User, time=0.7374, value=0}]
        }}})
    )
    Particles.RibbonColor3:LinearColor=&MultiplyLinearColorByVectorAndFloat(
     Color to Scale:LinearColor=$Particles.Initial.R2 Color
     RGB Scale:Vector3f=&VectorFromFloat(Value=6)
     Alpha Scale:NiagaraFloat=&FloatFromCurve(FloatCurve=DataInterface() {propertyValues {Curve {
         defaultValue=1
         keys [{interpMode=RCIM_Cubic, tangentMode=RCTM_User, time=0, value=1}, {interpMode=RCIM_Cubic, tangentMode=RCTM_User, time=0.7374, value=0}]
        }}})
    )
   ) {isSetParameters=true, module=SetVariables_#3}
  ]
  rendererClasses [&NiagaraRibbonRendererProperties]
  renderers [
   {rendererClass=&NiagaraRibbonRendererProperties}
   {rendererClass=&NiagaraRibbonRendererProperties, rendererIndex=1}
   {rendererClass=&NiagaraRibbonRendererProperties, rendererIndex=2}
   {rendererClass=&NiagaraRibbonRendererProperties, rendererIndex=3}
   {rendererClass=&NiagaraRibbonRendererProperties, rendererIndex=4}
  ]
 }
}
summary {
 systemName=NS_Boost_HexFade_v2
 emitters [
  {emitterName=HexRed, rendererClasses [&NiagaraSpriteRendererProperties]}
  {emitterName=BlueTrails001, rendererClasses [&NiagaraRibbonRendererProperties]}
  {emitterName=RedTrails001, rendererClasses [&NiagaraRibbonRendererProperties]}
 ]
 userVariables [
  User.bIsFrontendPreview {type=NiagaraBool}
  User.bIsPainted {type=NiagaraBool}
  User.ColorVariantID {type=NiagaraInt32}
  User.PaintedColor {type=LinearColor, description="", defaultValue {struct=LinearColor, value=<rgba 1 1 1 1>}}
  User.PaintedScalar {type=NiagaraFloat, defaultValue=1, description=""}
  User.PortCount {type=NiagaraInt32, defaultValue=1}
 ]
}
