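# Unlit translucent Niagara sprite material for the hex shapes of the HexFade vehicle boost cosmetic.
# A flipbook texture on UV1 supplies two channel masks, selected by the floored TextureSwitch dynamic parameter.
# A particle-color intensity estimate, counting red twice and omitting green, drives threshold corrections to tint and brightness.
# Dynamic Brightness scales the masked color into MF_FX_EmissiveAdjustments; particle alpha times the mask forms opacity.
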Material /VehicleCosmetics/Boosts/Boost_HexFade/Material/M_Boost_HexFade_Hex
graph {
 properties {bUsedWithNiagaraSprites=true, blendMode=BLEND_Translucent, pixelDepthOffsetMode=PDOM_Legacy, shadingModel=MSM_Unlit, usedShadingModels=Unlit}
 nodes [
  n1 = DynamicParameter() {paramNames [ColorShift, Brightness, WhiteValue, TextureSwitch]}
  n2 = ParticleColor()
  n3 = Multiply(A=@n2.RGB, B=Constant() {r=0.85}) {constB=0.85}
  n4 = Constant() {r=0.1}
  n5 = Add(A=@n2.R, B=@n2.R)
  n6 = Saturate(None=Multiply(A=Add(A=@n5, B=@n2.B), B=Constant() {r=0.3333}) {constB=0.3333})
  n7 = Multiply(A=Add(A=@n2.RGB, B=Constant() {r=0.08}) {constB=0.08}, B=Constant() {r=2})
  n8 = LinearInterpolate(
   A=If(A=@n6, B=Constant() {r=0.8}, "A > B"=@n3, "A == B"=@n3, "A < B"=Add(A=@n2.RGB, B=@n4) {constB=0.1})
   B=If(A=@n6, B=Constant() {r=0.1}, "A > B"=@n2.RGB, "A == B"=@n7, "A < B"=@n7)
   Alpha=@n2.RGB
  )
  n9 = TextureSample(UVs=TextureCoordinate() {coordinateIndex=1, desc=1x1}) {texture=/VehicleCosmetics/Boosts/Boost_HexFade/Texture/T_VCVFX_Boost_HexFade_FB_01}
  n10 = Saturate(None=LinearInterpolate(
    A=StaticComponentMaskParameter(None=@n9.RGB) {defaultR=true, parameterName=FlipbookChannel_01}
    B=StaticComponentMaskParameter(None=@n9.RGB) {defaultG=true, parameterName=FlipbookChannel_02}
    Alpha=Floor(None=@n1.TextureSwitch)
   ))
  n11 = Constant() {r=0.8}
  n12 = ScalarParameter() {defaultValue=2, parameterName=Brightness}
  n13 = Multiply(A=Constant() {r=0.85}, B=@n12) {constB=0.85}
  n14 = Add(A=@n12, B=Constant() {r=0.1}) {constB=0.1}
  n15 = Constant() {r=0.1}
  n16 = Multiply(A=Add(A=@n12, B=Constant() {r=0.08}) {constB=0.08}, B=Constant() {r=2}) {constB=2}
  n17 = Multiply(
   A=@n10
   B=LinearInterpolate(A=If(A=@n6, B=@n11, "A > B"=@n13, "A == B"=@n13, "A < B"=@n14), B=If(A=@n6, B=@n15, "A > B"=@n12, "A == B"=@n16, "A < B"=@n16), Alpha=@n6)
  )
  MaterialFunctionCall(Emissive=Multiply(A=@n1.Brightness, B=Multiply(A=@n8, B=@n17))) {materialFunction=&MF_FX_EmissiveAdjustments}
  Saturate(None=Multiply(A=@n2.A, B=@n10))
 ]
}
