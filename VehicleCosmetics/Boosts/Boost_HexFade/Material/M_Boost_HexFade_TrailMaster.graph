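# Translucent unlit master material for Niagara ribbon and sprite trails in the HexFade vehicle boost.
# A trail mask texture, optionally flipbook-animated, supplies switchable channel blends that can be modulated by a panning "lightning" mask.
# Particle color, adjusted by an R+R+B luminance estimate, and dynamic parameters for brightness, saturation and blend drive emissive through MF_FX_EmissiveAdjustments.
# Opacity combines the mask, particle alpha, optional depth fade and a power-curve fade along U.
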Material /VehicleCosmetics/Boosts/Boost_HexFade/Material/M_Boost_HexFade_TrailMaster
graph {
 properties {
  bUsedWithNiagaraRibbons=true
  bUsedWithNiagaraSprites=true
  blendMode=BLEND_Translucent
  pixelDepthOffsetMode=PDOM_Legacy
  shadingModel=MSM_Unlit
  usedShadingModels=Unlit
 }
 nodes [
  n1 = ParticleColor()
  n2 = Add(A=@n1.R, B=@n1.R)
  n3 = Saturate(None=Multiply(A=Add(A=@n2, B=@n1.B), B=Constant() {r=0.3333}) {constB=0.3333})
  n4 = If(A=@n3, B=Constant() {r=0.1}, "A > B"=@n1.RGB, "A == B"=@n1.RGB, "A < B"=Multiply(A=@n1.RGB, B=Constant() {r=1.5}) {constB=1.5}) {constB=0.1}
  n5 = If(A=@n3, B=Constant() {r=0.98}, "A > B"=@n1.RGB, "A == B"=@n1.RGB, "A < B"=Add(A=@n1.RGB, B=Constant() {r=0.05}) {constB=0.05}) {constB=0.93}
  n6 = DynamicParameter() {paramNames [Brightness,LerpBlend,Saturation,ColorShift]}
  n7 = Saturate(None=Add(A=LinearInterpolate(A=@n4, B=@n5), B=@n6.Saturation))
  n8 = TextureCoordinate() {uTiling=4, vTiling=4}
  n9 = ComponentMask(None=StaticSwitchParameter(True=TextureCoordinate(), False=Frac(None=@n8)) {defaultValue=true, parameterName=FlipBook}) {r=true}
  n10 = OneMinus(None=@n9)
  n11 = ScalarParameter() {defaultValue=1, parameterName=PanTile}
  n12 = Multiply(A=TextureCoordinate() {uTiling=0.05, vTiling=0.25}, B=@n11)
  n13 = Multiply(A=Time(), B=ScalarParameter() {defaultValue=1, parameterName=PanSpeed})
  n14 = TextureCoordinate() {uTiling=0.25, vTiling=0.05}
  n15 = StaticSwitchParameter(
   True=Rotator(Coordinate=Panner(Coordinate=@n12, Time=@n13) {speedY=0.4}, Time=Constant() {r=0.5}) {speed=3.14}
   False=Panner(Coordinate=Multiply(A=@n14, B=@n11), Time=@n13) {speedX=0.4}
  ) {parameterName=RotateFlipbook}
  n16 = Panner(Coordinate=TextureCoordinate() {coordinateIndex=1, vTiling=0.5}, Time=Constant() {r=1}) {speedY=0.25}
  n17 = Constant() {r=0.5}
  n18 = ScalarParameter() {defaultValue=4, parameterName=Columns}
  n19 = ScalarParameter() {defaultValue=4, parameterName=Rows}
  n20 = AppendVector(A=@n18, B=@n19)
  n21 = Divide(A=StaticSwitchParameter(True=Rotator(Coordinate=@n16, Time=@n17) {speed=3.14}, False=@n16) {parameterName=RotateFlipbook}, B=@n20)
  n22 = AppendVector(A=Multiply(A=@n18, B=@n19), B=@n19)
  n23 = Frac(None=Multiply(A=Time(), B=ScalarParameter() {defaultValue=1, parameterName=Speed}))
  n24 = AppendVector(A=@n23, B=@n23)
  n25 = Divide(A=Floor(None=Multiply(A=@n22, B=@n24)), B=@n20)
  n26 = TextureCoordinate() {coordinateIndex=1}
  n27 = Rotator(Coordinate=@n26, Time=Constant() {r=0.5}) {speed=3.14}
  n28 = TextureSample(UVs=StaticSwitchParameter(True=Add(A=@n21, B=@n25), False=StaticSwitchParameter(True=@n27, False=@n26) {parameterName=RotateFlipbook}) {defaultValue=true, parameterName=FlipBook}) {texture=/DealershipSharedLibrary/FX/Textures/Gradient_Masks/T_VCVFX_BasicTrail_A}
  n29 = StaticComponentMaskParameter(None=@n28.RGB) {defaultG=true, parameterName=Blend_01}
  n30 = StaticComponentMaskParameter(None=@n28.RGB) {defaultB=true, parameterName=Blend_02}
  n31 = Multiply(A=@n29, B=@n30)
  n32 = Add(A=@n29, B=@n30)
  n33 = StaticSwitchParameter(
   True=StaticSwitchParameter(
    True=LinearInterpolate(A=@n29, B=@n30, Alpha=@n6.LerpBlend)
    False=StaticSwitchParameter(True=@n31, False=@n32) {parameterName="BlendType(T=M)(F=A)"}
   ) {parameterName=LerpBlend}
   False=StaticComponentMaskParameter(None=@n28.RGB) {defaultG=true, parameterName=RGB_Mask}
  ) {parameterName=Blend}
  n34 = Multiply(
   A=Multiply(
    A=ScalarParameter() {defaultValue=3, parameterName=PanBrightness}
    B=TextureSample(UVs=@n15) {texture=/DealershipSharedLibrary/FX/Textures/Gradient_Masks/T_VCVFX_TimeMask_A}.RGB
   )
   B=@n33
  )
  n35 = Add(A=@n34, B=@n33)
  n36 = Multiply(
   A=StaticSwitchParameter(True=StaticSwitchParameter(True=@n35, False=@n34) {parameterName=AddPan?}, False=@n33) {parameterName=PanLightning}
   B=@n9
  )
  n37 = Saturate(None=Multiply(A=Multiply(A=@n36), B=@n10))
  n38 = Multiply(A=Multiply(A=Constant() {r=8}, B=Saturate(None=@n10)) {constB=0}, B=@n37)
  MaterialFunctionCall(Emissive=Multiply(A=DynamicParameter() {paramNames [Brightness,LerpBlend,Saturation,ColorShift]}.Brightness, B=Multiply(A=@n7, B=@n38))) {materialFunction=&MF_FX_EmissiveAdjustments}
  n39 = Multiply(A=@n37, B=ParticleColor().A)
  n40 = Multiply(
   A=@n39
   B=MaterialFunctionCall(Opacity In=Constant() {r=1}, DepthFade Distance=Constant() {r=100}) {materialFunction=&MF_VCVFX_DepthFade}.Result
  )
  n41 = ComponentMask(None=TextureCoordinate()) {r=true}
  Multiply(
   A=Saturate(None=StaticSwitchParameter(True=@n40, False=@n39) {parameterName=DBA?})
   B=Power(Base=OneMinus(None=@n41), Exp=ScalarParameter() {defaultValue=10, parameterName=PowerParameterFade})
  )
  Reroute(None=@n6.Brightness)
 ]
}
